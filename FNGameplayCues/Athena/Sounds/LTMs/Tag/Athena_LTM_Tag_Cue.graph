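# Sound cue for the Tag sound in an Athena limited-time mode, routed through the traps sound class.
# A Random node picks one of three equally weighted Tag wave variations.
# The pick passes through a default Modulator for slight pitch and volume variation, then a short fixed Delay at the root.
# It borrows the Athena trap-activation attenuation for distance falloff.
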SoundCue /FNGameplayCues/Athena/Sounds/LTMs/Tag/Athena_LTM_Tag_Cue
root=Delay_0
nodes [
 {id=Delay_0, children [Modulator_0], properties {delayMax=0.1, delayMin=0.1}}
 {id=Modulator_0, children [Random_0]}
 {id=Random_0, children [WavePlayer_1,WavePlayer_2,WavePlayer_3], properties {weights [1,1,1]}}
 {id=WavePlayer_1, properties {soundWaveAssetPtr=&/Athena_LTM_Tag_01}}
 {id=WavePlayer_2, properties {soundWaveAssetPtr=&/Athena_LTM_Tag_02}}
 {id=WavePlayer_3, properties {soundWaveAssetPtr=&/Athena_LTM_Tag_03}}
]
properties {
 attenuationSettings=/Game/Sounds/Attenuation/Fort_Traps_Activate_Athena_Att
 maxDistance=3500
 soundClassObject=&Fort_Traps
 volumeMultiplier=1
}
